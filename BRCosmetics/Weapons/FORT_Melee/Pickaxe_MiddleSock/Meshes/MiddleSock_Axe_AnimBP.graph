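# Animation Blueprint for the MiddleSock pickaxe axe mesh, posing its three dynamic claw bones.
# The input pose gets additive bone-space claw rotations and a translation, is cached as Base, then passes through zero-offset modifiers on the same bones to the output.
# Each update sets Not_PlayingMelee from the owner's claw harvesting montage, choosing the owner anim instance by front-end state.
# The AnimGraph shows no visible read of that flag.
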AnimBlueprint /BRCosmetics/Weapons/FORT_Melee/Pickaxe_MiddleSock/Meshes/MiddleSock_Axe_AnimBP
target_skeleton=&/MiddleSock_Axe_Skeleton
graphs [
 AnimGraph {nodes [
   n1 = "Transform (Modify) Bone - Bone: dyn_claw_01_BIND_jnt"!(
    ComponentPose=Local To Component(LocalPose=Use cached pose 'Base'())
    bAlphaBoolEnabled:bool=True
    Translation="0,0,0"
    Rotation="0,0,0"
    Scale="1,1,1"
   ) -> (Pose) {
    alphaInputType=Bool
    rotationMode=BMM_Additive
    rotationSpace=BCS_BoneSpace
    translationMode=BMM_Additive
    translationSpace=BCS_BoneSpace
    alphaBoolBlend {blendInTime=0.3, blendOption=Linear, blendOutTime=0.5}
    boneToModify {boneName=dyn_claw_01_BIND_jnt}
   }
   n2 = "Transform (Modify) Bone - Bone: dyn_claw_02_BIND_jnt"!(ComponentPose=@n1, bAlphaBoolEnabled:bool=True, Translation="0,0,0", Rotation="0,0,0", Scale="1,1,1") -> (Pose) {
    alphaInputType=Bool
    rotationMode=BMM_Additive
    rotationSpace=BCS_BoneSpace
    translationMode=BMM_Additive
    translationSpace=BCS_BoneSpace
    alphaBoolBlend {blendInTime=0.3, blendOption=Linear, blendOutTime=0.5}
    boneToModify {boneName=dyn_claw_02_BIND_jnt}
   }
   n3 = "Transform (Modify) Bone - Bone: dyn_claw_MOVE_BIND_jnt"!(ComponentPose=@n2, bAlphaBoolEnabled:bool=True, Translation="0,0,0", Rotation="0,0,0", Scale="1,1,1") -> (Pose) {
    alphaInputType=Bool
    rotationMode=BMM_Additive
    rotationSpace=BCS_BoneSpace
    translationMode=BMM_Additive
    translationSpace=BCS_BoneSpace
    alphaBoolBlend {blendInTime=0.3, blendOption=Linear, blendOutTime=0.5}
    boneToModify {boneName=dyn_claw_MOVE_BIND_jnt}
   }
   Output Pose(Result=Component To Local(ComponentPose=@n3)) {layerGroup=None}
   AnimGraphNode_SubInput_0 = Input Pose()
   n4 = "Transform (Modify) Bone - Bone: dyn_claw_01_BIND_jnt"(
    ComponentPose=Local To Component(LocalPose=@AnimGraphNode_SubInput_0)
    bAlphaBoolEnabled=True
    Translation="0,0,0"
    Rotation="0.000,-99.809,0.000"
    Scale="1,1,1"
   ) {
    alphaInputType=Bool
    rotationMode=BMM_Additive
    rotationSpace=BCS_BoneSpace
    translationMode=BMM_Additive
    translationSpace=BCS_BoneSpace
    alphaBoolBlend {blendInTime=0.3, blendOption=Sinusoidal, blendOutTime=0.2}
    boneToModify {boneName=dyn_claw_01_BIND_jnt}
    rotation {yaw=-99.809}
   }
   n5 = "Transform (Modify) Bone - Bone: dyn_claw_02_BIND_jnt"(ComponentPose=@n4, bAlphaBoolEnabled=True, Translation="0,0,0", Rotation="-0.000,109.817,0.000", Scale="1,1,1") {
    alphaInputType=Bool
    rotationMode=BMM_Additive
    rotationSpace=BCS_BoneSpace
    translationMode=BMM_Additive
    translationSpace=BCS_BoneSpace
    alphaBoolBlend {blendInTime=0.3, blendOption=Sinusoidal, blendOutTime=0.2}
    boneToModify {boneName=dyn_claw_02_BIND_jnt}
    rotation {yaw=109.817}
   }
   n6 = "Transform (Modify) Bone - Bone: dyn_claw_MOVE_BIND_jnt"(ComponentPose=@n5, bAlphaBoolEnabled=True, Translation="-4.310,-0.377,0.000", Rotation="0,0,0", Scale="1,1,1") {
    alphaInputType=Bool
    rotationMode=BMM_Additive
    rotationSpace=BCS_BoneSpace
    translationMode=BMM_Additive
    translationSpace=BCS_BoneSpace
    alphaBoolBlend {blendInTime=0.3, blendOption=Sinusoidal, blendOutTime=0.2}
    boneToModify {boneName=dyn_claw_MOVE_BIND_jnt}
    translation=<-4.31 -0.377>
   }
   Save cached pose 'Base'(Pose=Component To Local(ComponentPose=@n6)) {cacheName=Base}
  ]}
 EventGraph {nodes [
   Event BlueprintUpdateAnimation()
   > n1 = Branch(Condition=Get bIsInFrontEnd())
   n2 = GetOwnerAnimInstance!(self:Custom Character Part Anim Instance Object Reference) -> (ReturnValue:Anim Instance Object Reference)
   > Set Not_PlayingMelee(Not_PlayingMelee:bool=NOT Boolean(A=Montage_IsPlaying(self=@n2, Montage=/Game/Animation/Game/MainPlayer/Combat/Melee/DualWield/Claw/Claw_Harvesting_Combo_M))) -> (Output_Get:bool)
   n3 = GetOwnerFortPlayerAnimInstance!(self:Custom Character Part Anim Instance Object Reference) -> (ReturnValue:Fort Player Anim Instance Object Reference)
   Set Not_PlayingMelee(
    execute=@n1.else
    Not_PlayingMelee=NOT Boolean(A=Montage_IsPlaying(self=@n3, Montage=&Claw_Harvesting_Combo_M))
   )
   TryGetPawnOwner()
  ]}
]
variables [Not_PlayingMelee {type=boolean}]
